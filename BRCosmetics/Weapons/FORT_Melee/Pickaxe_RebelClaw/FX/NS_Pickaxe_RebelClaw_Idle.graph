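# Niagara system for the RebelClaw pickaxe's idle electricity, with bursts suppressed while User.bisSwinging is true.
# Electricity steps through User.LightningOrder by loop count, using spawn group to pick mesh and position, with randomized direction and travel.
# Electricity_Additional rolls bursts at three fixed local offsets with random meshes.
# Both use random loop durations, mesh and sprite renderers, Variant-based visibility tags, and flash-decay color curves scaled by time of day.
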NiagaraSystem /BRCosmetics/Weapons/FORT_Melee/Pickaxe_RebelClaw/FX/NS_Pickaxe_RebelClaw_Idle
dependencies {
 usedDataInterfaces [&NiagaraDataInterfaceCurve]
 usedDynamicInputs []
 usedModules [
  &self:SystemSpawnScript.SystemScriptSource.SystemScriptGraph.NiagaraNodeAssignment_0.SetVariables_#1
  &SystemState
 ]
}
emitters {
 Electricity {
  emitterUpdateScript [
   &EmitterState(
    Inactive Response=::NewEnumerator1
    Life Cycle Mode=::NewEnumerator1
    Loop Delay=3.4
    Loop Duration=&RandomRangeFloat(Minimum=0.1, Maximum=1.2, Evaluation Type=::NewEnumerator1)
    Recalculate Duration Each Loop=true
    UseLoopDelay=false
   )
   &SpawnBurst_Instantaneous(
    Spawn Group:NiagaraInt32=&SelectIntFromArray(Direct Array Index=$Emitter.LoopCount, Direct Array Mode=::NewEnumerator1, Int Selection Array=$User.LightningOrder)
    Spawn Probability=&MakeCustomFloatFromBool(Bool=$User.bisSwinging, True Float=0, False Float=1)
    Use Spawn Probability=true
    UseLoopCountLimit=false
   )
  ]
  particleSpawnScript [
   /SetVariables(
    Particles.ElectricityVisibilityTag:NiagaraInt32=&MakeIntFromFloat(Float=$User.Variant)
    Particles.FakeLightSize:Vector2f=&Vector2DFromFloat(Value=16)
    Particles.IsReverseDirection:NiagaraBool=&RandomBool()
    Particles.IsTravelingLightning:NiagaraBool=&RandomBool()
    Particles.MeshIndex=&SpawnGroup()
   ) {isSetParameters=true, module=SetVariables_#2}
   &DynamicMaterialParameters(
    Index 0 Param 1=&MakeCustomFloatFromBool(Bool=$Particles.IsReverseDirection, True Float=1, False Float=0)
    Index 0 Param 2=&MakeCustomFloatFromBool(Bool=$Particles.IsTravelingLightning, True Float=2.3, False Float=6)
    Index 0 Param 3=&FN_FrontEndFloatModifier(Gameplay=1, Front End=0.7)
    Module.Index0Param0Write=true
    Module.Index0Param1Write=true
    Module.Index0Param2Write=true
    Module.Index0Param3Write=false
   )
   &InitializeParticle(
    Color=<rgba 0.4 4 4 0.0025>
    Lifetime Max=1.1
    Lifetime Min=0.8
    Lifetime Mode=::NewEnumerator1
    Position=&Lerp_Vector(
     A=&SelectVectorFromArray(
      Direct Array Index=&SpawnGroup()
      Direct Array Mode=::NewEnumerator1
      Vector Selection Array=$User.LightningPositions
     )
     B=&SelectVectorFromArray(
      Direct Array Index=&SpawnGroup()
      Direct Array Mode=::NewEnumerator1
      Vector Selection Array=$User.LightningPositionsReversed
     )
     Alpha=&MakeCustomFloatFromBool(Bool=$Particles.IsReverseDirection, True Float=0, False Float=1)
    )
    Position Mode=::NewEnumerator1
    Sprite Size Mode=::NewEnumerator3
    Uniform Sprite Size=22
    UsePositionOffset=false
    Write Material Random=true
   )
   &AddVelocity(
    Cone Axis=<0 1 0>
    Module.UseConeFalloff=false
    Module.UseVelDistribution=false
    Override Seed=false
    Velocity Mode=::NewEnumerator2
    Velocity Speed=666
   )
   &ApplyInitialForces()
  ]
  particleUpdateScript [
   &ParticleState()
   &ScaleColor(
    Scale RGB=&VectorFromFloat(Value=&FloatFromCurve(
      FloatCurve=DataInterface() {propertyValues {Curve {keys [
          <at 0 0 RCIM_Constant>
          <at 0.027 8 RCIM_Constant>
          {interpMode=RCIM_Cubic, tangentMode=RCTM_User, time=0.0555, value=0.25}
          {arriveTangentWeight=0.3203, interpMode=RCIM_Cubic, tangentMode=RCTM_User, time=0.7, value=0}
         ]}}}
      Scale Curve=&Multiply_Float(A=&FNTimeOfDayScalar(Night Scale=0.7, Morning Scale=0.75, Evening Scale=0.75), B=0.9)
     ))
    ScaleA=false
    ScaleRGB=true
   )
   &CameraOffset(Camera Offset Amount=10)
  ]
  rendererClasses [&NiagaraMeshRendererProperties, &NiagaraSpriteRendererProperties]
  renderers [
   {rendererClass=&NiagaraMeshRendererProperties, rendererIndex=0}
   {rendererClass=&NiagaraMeshRendererProperties, rendererIndex=1}
   {rendererClass=&NiagaraMeshRendererProperties, rendererIndex=2}
   {rendererClass=&NiagaraMeshRendererProperties, rendererIndex=3}
   {rendererClass=&NiagaraSpriteRendererProperties, rendererIndex=4}
   {rendererClass=&NiagaraSpriteRendererProperties, rendererIndex=5}
  ]
 }
 Electricity_Additional {
  emitterUpdateScript [
   &EmitterState(
    Inactive Response=::NewEnumerator1
    Life Cycle Mode=::NewEnumerator1
    Loop Delay=3.4
    Loop Duration=&RandomRangeFloat(Minimum=0.066, Maximum=0.999, Evaluation Type=::NewEnumerator1)
    Recalculate Duration Each Loop=true
    UseLoopDelay=false
   )
   &SpawnBurst_Instantaneous(
    Spawn Probability=&Multiply_Float(A=0.5, B=&MakeCustomFloatFromBool(Bool=$User.bisSwinging, True Float=0, False Float=1))
    Use Spawn Probability=true
    UseLoopCountLimit=false
   )
   &SpawnBurst_Instantaneous(
    Spawn Group=1
    Spawn Probability=&Multiply_Float(A=0.6, B=&MakeCustomFloatFromBool(Bool=$User.bisSwinging, True Float=0, False Float=1))
    Use Spawn Probability=true
    UseLoopCountLimit=false
   ) {module=SpawnBurst_Instantaneous001}
   &SpawnBurst_Instantaneous(
    Spawn Group=2
    Spawn Probability=&Multiply_Float(A=0.35, B=&MakeCustomFloatFromBool(Bool=$User.bisSwinging, True Float=0, False Float=1))
    Use Spawn Probability=true
    UseLoopCountLimit=false
   ) {module=SpawnBurst_Instantaneous002}
  ]
  particleSpawnScript [
   &InitializeParticle(
    Color=<rgba 0.4 4 4 0.0025>
    Lifetime Max=1.1
    Lifetime Min=0.8
    Lifetime Mode=::NewEnumerator1
    Mesh Scale Mode=::NewEnumerator3
    Mesh Uniform Scale=&RandomRangeFloat(Minimum=0.75, Maximum=1)
    Position Mode=::NewEnumerator1
    Sprite Size Mode=::NewEnumerator3
    Uniform Sprite Size=22
    UsePositionOffset=false
    Write Material Random=true
   )
   &ShapeLocation(
    MaskSpawnGroup=true
    Non Uniform Scale=<1 2 1>
    Offset=<0 5 46.795>
    Offset Coordinate Space=::Local
    Offset Mode=::NewEnumerator0
    Override Seed=false
    Sphere Radius=2
    UseOwnerScale=true
   )
   &ShapeLocation(
    MaskSpawnGroup=true
    Offset=<0 12 -60.6432>
    Offset Coordinate Space=::Local
    Offset Mode=::NewEnumerator0
    Override Seed=false
    Spawn Group Mask=1
    Sphere Radius=2
    UseOwnerScale=true
   ) {module=ShapeLocation001}
   &ShapeLocation(
    MaskSpawnGroup=true
    Non Uniform Scale=<1 6 1>
    Offset=<0 0 59.153>
    Offset Coordinate Space=::Local
    Offset Mode=::NewEnumerator0
    Override Seed=false
    Spawn Group Mask=2
    Sphere Radius=2
    UseOwnerScale=true
   ) {module=ShapeLocation002}
   /SetVariables(
    Particles.ElectricityVisibilityTag:NiagaraInt32=&MakeIntFromFloat(Float=$User.Variant)
    Particles.FakeLightSize:Vector2f=&Vector2DFromFloat(Value=20)
    Particles.IsReverseDirection:NiagaraBool=&RandomBool()
    Particles.MeshIndex=&RandomRangeInt(Minimum=0, Maximum=1)
   ) {isSetParameters=true, module=SetVariables_#2}
   &DynamicMaterialParameters(
    Index 0 Param 1=&MakeCustomFloatFromBool(Bool=$Particles.IsReverseDirection, True Float=1, False Float=0)
    Index 0 Param 2=2.3
    Index 0 Param 3=&FN_FrontEndFloatModifier(Gameplay=1, Front End=0.7)
    Module.Index0Param0Write=true
    Module.Index0Param1Write=true
    Module.Index0Param2Write=true
    Module.Index0Param3Write=false
   )
   &InitialMeshOrientation(Use Rotation Vector=false)
   &AddVelocity(
    Cone Axis=<0 1 0>
    Module.UseConeFalloff=false
    Module.UseVelDistribution=false
    Override Seed=false
    Velocity Mode=::NewEnumerator2
    Velocity Speed=666
   )
   &ApplyInitialForces()
  ]
  particleUpdateScript [
   &ParticleState()
   &ScaleColor(
    Scale RGB=&VectorFromFloat(Value=&FloatFromCurve(
      FloatCurve=DataInterface() {propertyValues {Curve {keys [
          <at 0 0 RCIM_Constant>
          <at 0.027 3 RCIM_Constant>
          {arriveTangent=-1.3556, interpMode=RCIM_Cubic, leaveTangent=-1.3556, tangentMode=RCTM_User, time=0.0555, value=0.25}
          {arriveTangentWeight=0.3203, interpMode=RCIM_Cubic, tangentMode=RCTM_User, time=0.7, value=0}
         ]}}}
      Scale Curve=&Multiply_Float(A=&FNTimeOfDayScalar(Night Scale=0.7, Morning Scale=0.75, Evening Scale=0.75), B=0.9)
     ))
    ScaleA=false
    ScaleRGB=true
   )
   &CameraOffset(Camera Offset Amount=7)
  ]
  rendererClasses [&NiagaraMeshRendererProperties, &NiagaraSpriteRendererProperties]
  renderers [
   {rendererClass=&NiagaraMeshRendererProperties, rendererIndex=0}
   {rendererClass=&NiagaraMeshRendererProperties, rendererIndex=1}
   {rendererClass=&NiagaraMeshRendererProperties, rendererIndex=2}
   {rendererClass=&NiagaraMeshRendererProperties, rendererIndex=3}
   {rendererClass=&NiagaraSpriteRendererProperties, rendererIndex=4}
   {rendererClass=&NiagaraSpriteRendererProperties, rendererIndex=5}
  ]
 }
}
summary {
 systemName=NS_Pickaxe_RebelClaw_Idle
 emitters [
  {
   emitterName=Electricity_Additional
   rendererClasses [&NiagaraMeshRendererProperties, &NiagaraSpriteRendererProperties]
  }
  {emitterName=Electricity, rendererClasses [&NiagaraMeshRendererProperties, &NiagaraSpriteRendererProperties]}
 ]
 userVariables [
  User.bIsFrontEnd {type=NiagaraBool}
  User.bIsFrontEndPreview {type=NiagaraBool}
  User.bisSwinging {type=NiagaraBool}
  User.LightningOrder {
   type=NiagaraDataInterfaceArrayInt32
   description=""
   defaultValue {
    struct=NiagaraExt_VariableValue_DataInterface
    value {
     dataInterface=&self:NiagaraDataInterfaceArrayInt32_0
     dataInterfaceClass=&NiagaraDataInterfaceArrayInt32
    }
   }
  }
  User.LightningPositions {
   type=NiagaraDataInterfaceArrayFloat3
   description=""
   defaultValue {
    struct=NiagaraExt_VariableValue_DataInterface
    value {
     dataInterface=&self:NiagaraDataInterfaceArrayFloat3_0
     dataInterfaceClass=&NiagaraDataInterfaceArrayFloat3
    }
   }
  }
  User.LightningPositionsReversed {
   type=NiagaraDataInterfaceArrayFloat3
   description=""
   defaultValue {
    struct=NiagaraExt_VariableValue_DataInterface
    value {
     dataInterface=&self:NiagaraDataInterfaceArrayFloat3_1
     dataInterfaceClass=&NiagaraDataInterfaceArrayFloat3
    }
   }
  }
  User.Variant {type=NiagaraFloat, defaultValue=0, description=""}
 ]
}
